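# Material function that turns incoming material attributes into a Neon City hologram look with glitches, scanlines and a see-through illusion.
# The Diffuse texture gets time-driven pixelation, chromatic aberration and a height-based tint.
# Fresnel blends, skylight cubemap samples boosted at night and sun passthrough shape base color and emissive, brightened for ray tracing.
# Dithered, distance-scaled noise drives pixel depth offset; quality switches simplify or drop costly effects on lower tiers.
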MaterialFunction /Game/Global/Materials/MaterialGym/Materials/NeonCityHologram/MF_NeonCityHologram
graph {nodes [
  n1 = GetMaterialAttributes(None=FunctionInput() {inputName=Material, inputType=FunctionInput_MaterialAttributes}) {attributeGetTypes [#1]}
  n2 = SetMaterialAttributes!(
   MaterialAttributes=@n1.MaterialAttributes
   BaseColor=NamedRerouteUsage()
   Metallic=NamedRerouteUsage()
   Specular=NamedRerouteUsage()
   Roughness=NamedRerouteUsage()
   Anisotropy
   EmissiveColor=NamedRerouteUsage()
   OpacityMask
   PixelDepthOffset=NamedRerouteUsage()
   ShadingModel=ShadingModel()
  ) {attributeSetTypes [
    #2
    #3
    #4
    #5
    #6
    #7
    #8
    #9
    #10
   ]}
  FunctionOutput(None=@n2)
  NamedRerouteDeclaration(None=MaterialFunctionCall!(
    Default=MaterialFunctionCall!(Scanline Amount) {materialFunction=/Game/Environments/Asteria/Sets/NeonCity/Techart/Holograms/NewHologramMaterial/MFScanlines}.Result
    Low/Mobile=Constant() {r=1}
   ) {materialFunction=/Game/Effects/Fort_Effects/Materials/_Functions/MF_LowMobile_PathingSwitch}.Result) {name=Scanlines, nodeColor=<rgba 1 0 0.1641 1>}
  n3 = ScalarParameter() {defaultValue=4, group=NeonCityHologram, parameterName=SeethroughSidesFres}
  n4 = NamedRerouteUsage()
  n5 = LinearInterpolate(
   B=ScalarParameter() {defaultValue=3, group=NeonCityHologram, parameterName=SeeThroughSideMax}
   Alpha=Saturate(None=Fresnel(ExponentIn=@n3, Normal=@n4) {baseReflectFraction=0})
  ) {constB=0}
  n6 = TextureCoordinate()
  n7 = ScalarParameter() {defaultValue=30, group=NeonCityHologram, parameterName=Tile}
  n8 = Multiply(A=Time(), B=ScalarParameter() {defaultValue=0.33, group=NeonCityHologram, parameterName=Glitch Speed})
  n9 = Multiply(A=Frac(None=@n8)) {constB=10}
  n10 = LinearInterpolate(A=Multiply(A=@n7) {constB=0.8}, B=@n7, Alpha=Divide(A=Round(None=@n9)) {constB=10})
  n11 = Round(None=Multiply(A=@n6, B=@n10) {constB=5})
  n12 = Multiply(A=Frac(None=NamedRerouteUsage())) {constB=0.58}
  n13 = TextureObjectParameter() {parameterName=Diffuse, texture=/Engine/EngineResources/Black_Low.Black}
  n14 = TextureSample(UVs=LinearInterpolate(A=@n6, B=Divide(A=@n11, B=@n10), Alpha=Round(None=@n12)), Tex=@n13)
  n15 = NamedRerouteUsage()
  n16 = Multiply(A=@n15, B=ScalarParameter() {defaultValue=0.5, group=NeonCityHologram, parameterName=Fresenel_Outter})
  n17 = Multiply(A=@n15, B=ScalarParameter() {defaultValue=0.2, group=NeonCityHologram, parameterName=Fresenel_Inner})
  n18 = Saturate(None=Fresnel(Normal=NamedRerouteUsage()) {baseReflectFraction=0, exponent=3})
  n19 = ScalarParameter() {defaultValue=0.8, group=NeonCityHologram, parameterName=MobileBrightness}
  n20 = Add(
   A=LinearInterpolate(A=@n14.RGB, B=NamedRerouteUsage(), Alpha=Saturate(None=@n5))
   B=QualitySwitch(Default=LinearInterpolate(A=@n16, B=@n17, Alpha=@n18), Low=Multiply(A=@n19, B=@n15))
  )
  n21 = NamedRerouteUsage()
  n22 = Saturate(None=CollectionParameter() {collection=&FortniteMaterialParameters, parameterName=SunLightColor})
  n23 = Multiply(
   A=ScalarParameter() {defaultValue=0.02, group=NeonCityHologram, parameterName=Sunlight Passthrough}
   B=Multiply(A=ComponentMask(None=@n22) {b=true, g=true, r=true}, B=NamedRerouteUsage())
  )
  n24 = Clamp(None=MaterialFunctionCall() {materialFunction=/Game/Packages/Fortress_Sky/Fortnite_SunScreenSpaceFalloff}.Dot*-1)
  n25 = Add(A=Clamp(None=Multiply(A=@n20, B=@n21)), B=QualitySwitch(Default=Multiply(A=@n23, B=@n24), Low=Constant()))
  NamedRerouteDeclaration(None=Multiply(A=@n25, B=ScalarParameter() {defaultValue=1, group=NeonCityHologram, parameterName=ColorNess})) {name=Basecolor_OUT, nodeColor=<rgba 0.125 1 0 1>}
  n26 = TextureSampleParameter2D() {parameterName=SRM, samplerType=SAMPLERTYPE_Masks, texture=/Game/Global/Textures/Default/Blanks/Temp_specular_Mask}
  NamedRerouteDeclaration(None=MaterialFunctionCall(Default=@n26.R, Low/Mobile=Constant() {r=0.5}) {materialFunction=&MF_LowMobile_PathingSwitch}.Result) {name=Specular_OUT, nodeColor=<rgba 0.2656 0 1 1>}
  n27 = MaterialFunctionCall(Float2=StaticSwitchParameter(
    True=MaterialFunctionCall(X=@n26.G, Y=@n26.G) {materialFunction=&MakeFloat2}.Result
    False=MaterialFunctionCall(X=@n26.B, Y=@n26.B) {materialFunction=&MakeFloat2}.Result
   ) {group=NeonCityHologram, parameterName=GreenBlueSwizzle}) {materialFunction=&BreakOutFloat2Components}
  NamedRerouteDeclaration(None=MaterialFunctionCall(Default=@n27.G, Low/Mobile=Constant() {r=0.5}) {materialFunction=&MF_LowMobile_PathingSwitch}.Result) {name=Roughness_OUT, nodeColor=<rgba 0.0547 1 0 1>}
  n28 = MaterialFunctionCall(Float3=OneMinus(None=CameraVectorWS())) {materialFunction=&BreakOutFloat3Components}
  n29 = Add(A=@n28.B, B=ScalarParameter() {defaultValue=-1, group=NeonCityHologram, parameterName=OffsetSeeThrough})
  n30 = Multiply(
   A=SkyLightEnvMapSample(
    Direction=MaterialFunctionCall(X=@n28.R, Y=@n28.R, Z=@n29) {materialFunction=&MakeFloat3}.Result
    Roughness=ScalarParameter() {defaultValue=0.1, group=NeonCityHologram, parameterName=HologramRoughness}
   )
   B=ScalarParameter() {defaultValue=0.15, group=NeonCityHologram, parameterName=CubemapBleed}
  )
  n31 = Constant() {r=1}
  n32 = MaterialFunctionCall(
   Value_Morning=@n31
   Value_Day=@n31
   Value_Evening=@n31
   Value_Night=ScalarParameter() {defaultValue=5, group=NeonCityHologram, parameterName=NightBoost}
  ) {materialFunction=&Scalar_Time_of_Day}
  n33 = Constant3Vector() {constant=<rgba 0.2031 0.1723 0.1315 1>}
  n34 = Constant3Vector() {constant=<rgba 0.0373 0.0718 0.1302 1>}
  NamedRerouteDeclaration(None=QualitySwitch(
    Default=Multiply(A=@n30, B=@n32.Result)
    Low=MaterialFunctionCall(Color_Morning=@n33, Color_Day=@n33, Color_Evening=@n33, Color_Night=@n34) {materialFunction=&Color_Time_of_Day}.Result
   )) {name=Cubemap, nodeColor=<rgba 0.8984 0 1 1>}
  n35 = Multiply(A=Time(), B=ScalarParameter() {defaultValue=0.02, group=NeonCityHologram, parameterName=GlitchSpeed})
  NamedRerouteDeclaration(None=Multiply(A=@n35) {constB=3}) {name=GlitchUV, nodeColor=<rgba 1 0.5859 0 1>}
  n36 = Time()
  n37 = ScalarParameter() {defaultValue=0.02, group=NeonCityHologram, parameterName=Glitch Speed Curve}
  n38 = CurveAtlasRowParameter(CurveTime=Frac(None=Multiply(A=@n36, B=@n37))) {
   atlas=/Game/Environments/Asteria/Sets/NeonCity/Techart/Holograms/NewHologramMaterial/HologramCurves
   curve=/Game/Environments/Asteria/Sets/NeonCity/Techart/Holograms/NewHologramMaterial/HologramCurve
   group=NeonCityHologram
   parameterName=Glitch
  }
  NamedRerouteDeclaration(None=@n38.B) {name=Glitch Blue, nodeColor=<rgba 1 0.5625 0 1>}
  n39 = LinearInterpolate(B=ScalarParameter() {defaultValue=0.007, parameterName=ChromaticAb}, Alpha=Clamp(None=Sine(None=@n8) {period=20}) {minDefault=-1})
  n40 = Constant()
  n41 = TextureSample(
   UVs=Add(A=@n6, B=MaterialFunctionCall(R=@n39, G=@n40, B=@n40) {materialFunction=&AppendMany}.RG)
   Tex=@n13
  ) {samplerSource=SSM_Wrap_WorldGroupSettings}
  n42 = Add(A=@n6, B=MaterialFunctionCall(R=@n40, G=@n39, B=@n40) {materialFunction=&AppendMany}.RG)
  n43 = Add(
   A=@n6
   B=MaterialFunctionCall(R=@n40, G=@n40, B=OneMinus(None=@n39)) {materialFunction=&AppendMany}.RG
  )
  n44 = QualitySwitch(
   Default=MaterialFunctionCall(
    R=@n41.R
    G=TextureSample(UVs=@n42, Tex=@n13) {samplerSource=SSM_Wrap_WorldGroupSettings}.G
    B=TextureSample(UVs=@n43, Tex=@n13) {samplerSource=SSM_Wrap_WorldGroupSettings}.B
   ) {materialFunction=&AppendMany}.RGB
   Low=@n41.RGB
   Medium=@n41.RGB
  )
  n45 = Round(None=NamedRerouteUsage())
  n46 = VectorParameter() {group=NeonCityHologram, parameterName=Overall Hologram Tint, defaultValue=<rgba 0.5677 0.4619 0.1286 1>}
  n47 = TransformPosition(None=WorldPosition().XYZ) {transformSourceType=TRANSFORMPOSSOURCE_World}
  n48 = MaterialFunctionCall() {materialFunction=&ObjectLocalBounds}
  n49 = LinearInterpolate(
   A=ScalarParameter() {defaultValue=0.5, group=NeonCityHologram, parameterName=Hologram Tint Min}
   B=ScalarParameter() {group=NeonCityHologram, parameterName=Hologram Tint Max}
   Alpha=MaterialFunctionCall(Float3=Divide(A=@n47, B=@n48.Local Bounds Size)) {materialFunction=&BreakOutFloat3Components}.B
  )
  NamedRerouteDeclaration(None=LinearInterpolate(A=LinearInterpolate(A=@n44, B=@n14.RGB, Alpha=@n45), B=@n46.RGB, Alpha=Saturate(None=@n49))) {name=Color, nodeColor=<rgba 0.3125 0 1 1>}
  n50 = ScalarParameter() {defaultValue=8, group=NeonCityHologram, parameterName=LumenBrightness}
  n51 = Multiply(A=@n25, B=ScalarParameter() {defaultValue=1, group=NeonCityHologram, parameterName=EmissiveNess})
  NamedRerouteDeclaration(None=RayTracingQualitySwitch(Normal=@n51, RayTraced=Multiply(A=@n51, B=@n50))) {name=Emissive_OUT, nodeColor=<rgba 1 0 0.4453 1>}
  n52 = Multiply(
   A=Multiply(
    A=ScalarParameter() {defaultValue=150, group=NeonCityHologram, parameterName=PDO_Distance}
    B=ScalarParameter() {defaultValue=1, group=NeonCityHologram, parameterName=PDoDistanceMult}
   )
   B=MaterialFunctionCall() {materialFunction=&DitherTemporalAA}.Result
  )
  n53 = MaterialFunctionCall(Float3=WorldPosition().XYZ) {materialFunction=&BreakOutFloat3Components}
  n54 = AppendVector(A=@n53.G, B=@n53.G)
  n55 = ScalarParameter() {defaultValue=50, group=NeonCityHologram, parameterName=GlitchScale}
  n56 = TextureSample(UVs=Add(A=@n35, B=Divide(A=@n54, B=@n55))) {
   samplerType=SAMPLERTYPE_Masks
   texture=/Game/Environments/Asteria/Sets/NeonCity/Techart/Holograms/F_MED_BlueBadass_NeonLines_-_GlitchNoise_v1
  }
  n57 = WorldPosition() {worldPositionShaderOffset=WPT_ExcludeAllShaderOffsets}
  n58 = CameraPositionWS()
  n59 = Divide(A=Add(A=Distance(A=@n57.XYZ, B=@n58), B=Constant() {r=-25000}), B=Constant() {r=50000})
  n60 = Constant() {r=4}
  n61 = Multiply(A=Multiply(A=Clamp(None=@n56.RGB) {minDefault=0.1}, B=Constant() {r=1500000}), B=Saturate(None=Power(Base=@n59, Exp=@n60)))
  n62 = Constant()
  NamedRerouteDeclaration(None=QualitySwitch(Default=Add(A=@n52, B=@n61), Low=@n62, Medium=@n62)) {name=PixelDepthOffset, nodeColor=<rgba 0 0.7656 1 1>}
  NamedRerouteDeclaration(None=Transform(None=@n1.Normal) {transformType=TRANSFORM_Local}) {name=Normal_IN, nodeColor=<rgba 1 0.6328 0 1>}
  NamedRerouteDeclaration(None=MaterialFunctionCall(Default=@n27.R, Low/Mobile=Constant()) {materialFunction=&MF_LowMobile_PathingSwitch}.Result) {name=Metallic_OUT, nodeColor=<rgba 1 0.5859 0 1>}
 ]}
